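# Item Override animation layers for the Galileo Lobster item on the Fortnite player skeleton, covering melee blocking and power aiming.
# ItemUpperBody runs a guard state machine that alternates between two variable-driven deflect sets.
# ItemFinalPoseOverride selects Shock or Push aim offsets, then twists the pelvis while restoring IK feet and blending back the upper body.
# Airborne and zipline additives switch on guarding; other layers handle double-jump falls, left-hand FK and root yaw correction.
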AnimBlueprint /Galileo/Items/Lobster/Assets/Animation/Galileo_LobsterLayerAnimBP
target_skeleton=&Fortnite_M_Avg_Player_Skeleton
graphs [
 AnimGraph {nodes [Output Pose() {layerGroup=None}]}
 EventGraph {links [], nodes []}
 ItemAimOffset {nodes [
   Input Pose!() -> (Pose, AimOffsetAlpha:double, Yaw:double, Pitch:double) {name=InPose_AimOffset}
   Input Pose!() -> (Pose) {name=InPose_PreAimOffsets}
   Output Pose() {layerGroup=Item Override}
  ]}
 ItemFallAdditive {nodes [
   Input Pose() {name=InPoseItemFallAdditive}
   n1 = Get FallPlayRate() -> (FallPlayRate:float)
   Output Pose(Result=Blend Poses by bool(
     BlendPose_0="Sequence Player (additive) 'M_JumpFall_ARDS_Additive'"!(PlayRate:float=@n1) -> (Pose) {bLoopAnimation=false, sequence=/Game/Animation/Game/MainPlayer/Locomotion/Medium/Male/Jump/AssaultRifle/M_JumpFall_ARDS_Additive}
     BlendPose_1="Sequence Player (additive) 'JumpFall_NoWep_CMM_Additive'"!(PlayRate:float=@n1) -> (Pose) {bLoopAnimation=false, sequence=/Game/Animation/Game/MainPlayer/Locomotion/Medium/Male/Jump/NoWep/JumpFall_NoWep_CMM_Additive}
     bActiveValue=Get bIsMeleeGuarding() -> (bIsMeleeGuarding:bool)
    ) {blendType=Linear}) {layerGroup=Item Override}
  ]}
 ItemFallLandAdditive {nodes [
   Input Pose() {name=InPoseItemFallLandAdditive}
   Output Pose() {layerGroup=Item Override}
  ]}
 ItemFallLevel2Override {nodes [
   AnimGraphNode_StateMachine_0 = Fall() {is=title AnimGraphNode_StateMachine, bAllowConduitEntryStates=true}
   Output Pose(Result=@AnimGraphNode_StateMachine_0) {layerGroup=Item Override}
   Save cached pose 'InAir'(Pose=Input Pose() {name=InPose_FallLevel2}) {cacheName=InAir}
  ]}
 Fall {
  entry=Conduit
  path=ItemFallLevel2Override.AnimGraphNode_StateMachine_0.Fall
  ids=id
  nodes [
   AnimStateConduitNode_0 = Conduit()
   AnimStateNode_0 = DefaultFall() {is=title AnimStateNode}
   AnimStateNode_1 = DoubleJumpFall() {is=title AnimStateNode}
   AnimStateTransitionNode_0 = AnimStateTransitionNode() {logicType=TLT_Inertialization, priorityOrder=100}
   AnimStateTransitionNode_1 = AnimStateTransitionNode() {logicType=TLT_Inertialization}
   AnimStateTransitionNode_2 = AnimStateTransitionNode()
  ]
  links [
   {
    dst=DefaultFall
    rule=ItemFallLevel2Override.AnimGraphNode_StateMachine_0.Fall.AnimStateTransitionNode_0.Transition
    src=Conduit
    via=AnimStateTransitionNode_0
   }
   {
    dst=DoubleJumpFall
    rule=ItemFallLevel2Override.AnimGraphNode_StateMachine_0.Fall.AnimStateTransitionNode_1.Transition
    src=Conduit
    via=AnimStateTransitionNode_1
   }
   {
    dst=Conduit
    rule=ItemFallLevel2Override.AnimGraphNode_StateMachine_0.Fall.AnimStateTransitionNode_2.Transition
    src=DefaultFall
    via=AnimStateTransitionNode_2
   }
  ]
 }
 Conduit {path=ItemFallLevel2Override.AnimGraphNode_StateMachine_0.Fall.AnimStateConduitNode_0.Conduit, nodes [Result(bCanEnterTransition=true)]}
 DefaultFall {
  path=ItemFallLevel2Override.AnimGraphNode_StateMachine_0.Fall.AnimStateNode_0.DefaultFall
  nodes [Output Animation Pose(Result=Use cached pose 'InAir'())]
 }
 DoubleJumpFall {
  path=ItemFallLevel2Override.AnimGraphNode_StateMachine_0.Fall.AnimStateNode_1.DoubleJumpFall
  nodes [Output Animation Pose(Result=Sequence Player 'NoWep_Land_Anticipation_Loop'() {sequence=/Game/Animation/Game/MainPlayer/Locomotion/Medium/Male/InAir/NoWep_Land_Anticipation_Loop})]
 }
 Transition {path=ItemFallLevel2Override.AnimGraphNode_StateMachine_0.Fall.AnimStateTransitionNode_0.Transition, nodes [Result(bCanEnterTransition=true)]}
 Transition {
  path=ItemFallLevel2Override.AnimGraphNode_StateMachine_0.Fall.AnimStateTransitionNode_1.Transition
  nodes [Result(bCanEnterTransition=Get bIsDoubleJumpFall() -> (bIsDoubleJumpFall:bool))]
 }
 Transition {path=ItemFallLevel2Override.AnimGraphNode_StateMachine_0.Fall.AnimStateTransitionNode_2.Transition, nodes [Result(bCanEnterTransition=true)]}
 ItemFinalPoseOverride {nodes [
   n1 = "Transform (Modify) Bone - Bone: pelvis"!(
    ComponentPose=Local To Component(LocalPose=Use cached pose 'FinalPoseCache'())
    Alpha=Get PelvisTwistAlpha() -> (PelvisTwistAlpha:float)
    Rotation_Roll:float=0
    Rotation_Pitch:float=0
    Rotation_Yaw:float=Get CustomPelvisTwist() -> (CustomPelvisTwist:float)
   ) -> (Pose) {
    rotationMode=BMM_Additive
    rotationSpace=BCS_WorldSpace
    translationSpace=BCS_WorldSpace
    alphaBoolBlend {blendOption=Linear}
    boneToModify {boneName=pelvis}
   }
   n2 = Layered blend per bone(
    BasePose=Layered blend per bone(BasePose=Component To Local(ComponentPose=@n1), BlendPoses_0=Use cached pose 'FinalPoseCache'()) {layerSetup [{branchFilters [{boneName=ik_foot_root}]}]}
    BlendPoses_0=Use cached pose 'FinalPoseCache'()
    BlendWeights_0=Get BlendLowerBodyFloat() -> (BlendLowerBodyFloat:double)
   ) {bMeshSpaceRotationBlend=true, layerSetup [{branchFilters [{boneName=Spine_01}, {boneName=ik_hand_root}]}]}
   Output Pose(Result=@n2) {layerGroup=Item Override}
   AnimGraphNode_StateMachine_1 = Power() {is=title AnimGraphNode_StateMachine, bAllowConduitEntryStates=true}
   n3 = Use cached pose 'Cached_InputPose'()
   n4 = Get bIsUsingPowerAO() -> (bIsUsingPowerAO:bool)
   n5 = Break Fort Anim Input Adjusted Aim!(FortAnimInput_AdjustedAim:"Fort Anim Input Adjusted Aim Structure (by ref)"=Get AdjustedAim() -> (AdjustedAim:Fort Anim Input Adjusted Aim Structure)) -> (ResultingYaw:float=~, ResultingPitch:float=~, ZeroOutPitchWeightCurveName:Name) {is=title K2Node_BreakStruct}
   n6 = AimOffset Player 'Shock_AO'!(X:float=@n5, Y:float=@n5.ResultingPitch, BasePose:Pose Link Structure=Use cached pose 'Cached_InputPose'(), Alpha:float=1) -> (Pose:Pose Link Structure) {blendSpace=/Galileo/Items/Lobster/Assets/Animation/Powers/Ambervine/Shock_AO, alphaBoolBlend {blendOption=Linear}}
   n7 = Use cached pose 'Cached_InputPose'()
   Save cached pose 'FinalPoseCache'(Pose=Blend Poses by bool(
     BlendPose_0=Layered blend per bone(BasePose=@n6, BlendPoses_0=@n7) {layerSetup [{branchFilters [{boneName=neck_01}]}]}
     BlendPose_1=Blend Poses by bool(BlendPose_0=@AnimGraphNode_StateMachine_1, BlendPose_1=@n3, BlendTime_0=0.25, BlendTime_1=0.4, bActiveValue=@n4) {blendType=Linear}
     BlendTime_0=0.2
     BlendTime_1=0.4
     bActiveValue=Get bAO_BlendIn() -> (bAO_BlendIn:bool)
    ) {blendType=Linear}) {cacheName=FinalPoseCache}
   Save cached pose 'Cached_InputPose'(Pose=Input Pose() {name=InFinalPose}) {cacheName=Cached_InputPose}
  ]}
 Power {
  entry=PowerAO
  path=ItemFinalPoseOverride.AnimGraphNode_StateMachine_1.Power
  ids=id
  nodes [AnimStateNode_3 = PowerAO() {is=title AnimStateNode}]
 }
 PowerAO {
  path=ItemFinalPoseOverride.AnimGraphNode_StateMachine_1.Power.AnimStateNode_3.PowerAO
  nodes [Output Animation Pose(Result=AimOffset Player 'Galileo_Lobster_Push_AO'!(
     X:float=Get ResultingYaw() -> (ResultingYaw:double)
     Y:float=Get ResultingPitch() -> (ResultingPitch:float)
     BasePose:Pose Link Structure=Use cached pose 'Cached_InputPose'()
     Alpha:float=1
    ) -> (Pose:Pose Link Structure) {blendSpace=/Galileo/Items/Lobster/Assets/Animation/Powers/Push/Galileo_Lobster_Push_AO, alphaBoolBlend {blendOption=Linear}})]
 }
 ItemFullBodyAimOffsets {nodes [
   Input Pose!() -> (Pose, FullBody_AimOffsetAlpha:double, Aim_Yaw:double, Aim_Pitch:double) {name=InPoseFullBodyAimOffsets}
   Input Pose!() -> (Pose) {name=InPosePreFullBodyAimOffsets}
   Output Pose() {layerGroup=Item Override}
  ]}
 ItemFullBodyOverride {nodes [
   Input Pose() {name=InPoseFullBody}
   Output Pose() {layerGroup=Item Override}
  ]}
 ItemIdleOverride {nodes [
   Input Pose() {name=InPose_Idle}
   Output Pose() {layerGroup=Item Override}
  ]}
 ItemJetPackJumpAdditive {nodes [Output Pose(Result=Blend Poses by bool(
     BlendPose_0="Sequence Player (additive) 'JetpackApex_ARDS_Loop_Additive'"() {sequence=/Game/Animation/Game/MainPlayer/Locomotion/Medium/Male/Jump/AssaultRifle/JetpackApex_ARDS_Loop_Additive}
     BlendPose_1="Sequence Player (additive) 'JetpackApex_NoWep__Loop_CMM_Additive'"() {sequence=/Game/Animation/Game/MainPlayer/Locomotion/Medium/Male/Jump/NoWep/JetpackApex_NoWep__Loop_CMM_Additive}
     BlendTime_1=1
     bActiveValue=Get bIsMeleeGuarding()
    ) {blendType=Linear}) {layerGroup=Item Override}]}
 ItemJetPackStartAdditive {nodes [Output Pose() {layerGroup=Item Override}]}
 ItemJumpLoopAdditive {nodes [
   Input Pose() {name=InPoseItemJumpLoopAdditive}
   Output Pose(Result=Blend Poses by bool(
     BlendPose_0="Sequence Player (additive) 'M_JumpApex_ARDS_Loop_Additive'"() {playRate=0.5, sequence=/Game/Animation/Game/MainPlayer/Locomotion/Medium/Male/Jump/AssaultRifle/M_JumpApex_ARDS_Loop_Additive}
     BlendPose_1="Sequence Player (additive) 'JumpApex_NoWep__Loop_CMM_Additive'"() {sequence=/Game/Animation/Game/MainPlayer/Locomotion/Medium/Male/Jump/NoWep/JumpApex_NoWep__Loop_CMM_Additive}
     bActiveValue=Get bIsMeleeGuarding()
    ) {blendType=Linear}) {layerGroup=Item Override}
  ]}
 ItemJumpOverride {nodes [
   Input Pose() {name=InPoseJump}
   Output Pose() {layerGroup=Item Override}
  ]}
 ItemJumpUpAdditive {nodes [
   Input Pose() {name=InPoseItemJumpUpAdditive}
   Output Pose(Result=Blend Poses by bool(
     BlendPose_0="Sequence Player (additive) 'M_JumpUp_ARDS_Additive'"() {bLoopAnimation=false, sequence=/Game/Animation/Game/MainPlayer/Locomotion/Medium/Male/Jump/AssaultRifle/M_JumpUp_ARDS_Additive}
     BlendPose_1="Sequence Player (additive) 'JumpUp_NoWep_CMM_Additive'"() {bLoopAnimation=false, sequence=/Game/Animation/Game/MainPlayer/Locomotion/Medium/Male/Jump/NoWep/JumpUp_NoWep_CMM_Additive}
     bActiveValue=Get bIsMeleeGuarding()
    ) {blendType=Linear}) {layerGroup=Item Override}
  ]}
 ItemPreAimOffsets {nodes [
   Input Pose() {name=InPosePreAimOffsets}
   Output Pose() {layerGroup=Item Override}
  ]}
 ItemPreIK {nodes [Output Pose(Result=Modify Curve(SourcePose=Input Pose() {name=InPose_PreIK}, CurveValues_0=0, Alpha=Get IsMeleeGuardingFloat() -> (IsMeleeGuardingFloat:double)) {curveNames [LeftHandFK], curveValues [0]}) {layerGroup=Item Override}]}
 ItemRightHandIK {nodes [
   Input Pose() {name=InPosePreRightHandIK}
   Input Pose() {name=InPosePostRightHandIK}
   Output Pose() {layerGroup=Item Override}
  ]}
 ItemSkeletalControl {nodes [
   n1 = Local To Component(LocalPose=Input Pose())
   n2 = Get RootRotationOffset() -> (RootRotationOffset:Rotator)
   Output Pose(Result=Component To Local(ComponentPose="Transform (Modify) Bone - Bone: Root"!(ComponentPose=@n1, AlphaCurveName:Name=EnableAccelYawCorrection, Rotation=@n2) -> (Pose) {alphaInputType=Curve, rotationMode=BMM_Additive, alphaBoolBlend {blendOption=Linear}, boneToModify {boneName=Root}})) {layerGroup=Item Override}
  ]}
 ItemSwimJumpFallLoop {nodes [Output Pose() {layerGroup=Item Override}]}
 ItemSwimJumpSurfaceEnd {nodes [Output Pose() {layerGroup=Item Override}]}
 ItemSwimJumpSurfaceLoop {nodes [Output Pose() {layerGroup=Item Override}]}
 ItemUpperBody {nodes [
   AnimGraphNode_StateMachine_0 = Block() {is=title AnimGraphNode_StateMachine}
   Output Pose(Result=@AnimGraphNode_StateMachine_0) {layerGroup=Item Override}
   Save cached pose 'UpperBody_CachePose'(Pose=Input Pose!() -> (Pose, InputParam:Fort Anim Input Adjusted Aim Structure) {name=InPoseUpperBody}) {cacheName=UpperBody_CachePose}
  ]}
 Block {
  entry=Default
  path=ItemUpperBody.AnimGraphNode_StateMachine_0.Block
  ids=id
  nodes [
   AnimStateConduitNode_0 = BackToDefault() {is=title AnimStateConduitNode}
   AnimStateNode_0 = GuardLoop() {is=title AnimStateNode, stateEntered {guid=#1, notifyName=GuardLoopEntered}}
   AnimStateNode_1 = DeflectEntry1() {is=title AnimStateNode}
   AnimStateNode_11 = Default() {is=title AnimStateNode}
   AnimStateNode_12 = GuardToIdle() {is=title AnimStateNode}
   AnimStateNode_2 = DeflectHold1() {is=title AnimStateNode, stateEntered {guid=#2, notifyName=DeflectHold1Entered}}
   AnimStateNode_3 = DeflectExit1() {is=title AnimStateNode}
   AnimStateNode_4 = DeflectTransition1_2() {is=title AnimStateNode}
   AnimStateNode_5 = GuardIntro() {is=title AnimStateNode}
   AnimStateNode_6 = DeflectHold2() {is=title AnimStateNode, stateEntered {guid=#3, notifyName=DeflectHold2Entered}}
   AnimStateNode_7 = DeflectTransition2_1() {is=title AnimStateNode}
   AnimStateNode_8 = DeflectExit2() {is=title AnimStateNode}
   AnimStateTransitionNode_0 = AnimStateTransitionNode() {crossfadeDuration=0.15}
   AnimStateTransitionNode_1 = AnimStateTransitionNode() {bAutomaticRuleBasedOnSequencePlayerInState=true, crossfadeDuration=0.1}
   AnimStateTransitionNode_10 = AnimStateTransitionNode()
   AnimStateTransitionNode_11 = AnimStateTransitionNode()
   AnimStateTransitionNode_12 = AnimStateTransitionNode() {crossfadeDuration=0.1}
   AnimStateTransitionNode_13 = AnimStateTransitionNode() {crossfadeDuration=0.1}
   AnimStateTransitionNode_14 = AnimStateTransitionNode()
   AnimStateTransitionNode_15 = AnimStateTransitionNode() {bAutomaticRuleBasedOnSequencePlayerInState=true}
   AnimStateTransitionNode_16 = AnimStateTransitionNode()
   AnimStateTransitionNode_17 = AnimStateTransitionNode() {crossfadeDuration=0.1}
   AnimStateTransitionNode_18 = AnimStateTransitionNode()
   AnimStateTransitionNode_19 = AnimStateTransitionNode() {bAutomaticRuleBasedOnSequencePlayerInState=true, crossfadeDuration=0.1}
   AnimStateTransitionNode_2 = AnimStateTransitionNode() {bAutomaticRuleBasedOnSequencePlayerInState=true, crossfadeDuration=0.1}
   AnimStateTransitionNode_20 = AnimStateTransitionNode() {bAutomaticRuleBasedOnSequencePlayerInState=true, crossfadeDuration=0.1}
   AnimStateTransitionNode_21 = AnimStateTransitionNode() {bAutomaticRuleBasedOnSequencePlayerInState=true, crossfadeDuration=0.1}
   AnimStateTransitionNode_22 = AnimStateTransitionNode()
   AnimStateTransitionNode_23 = AnimStateTransitionNode()
   AnimStateTransitionNode_3 = AnimStateTransitionNode() {bAutomaticRuleBasedOnSequencePlayerInState=true, crossfadeDuration=0.15}
   AnimStateTransitionNode_4 = AnimStateTransitionNode() {bAutomaticRuleBasedOnSequencePlayerInState=true, crossfadeDuration=0.1}
   AnimStateTransitionNode_5 = AnimStateTransitionNode() {bAutomaticRuleBasedOnSequencePlayerInState=true, crossfadeDuration=0.1}
   AnimStateTransitionNode_6 = AnimStateTransitionNode()
   AnimStateTransitionNode_7 = AnimStateTransitionNode()
   AnimStateTransitionNode_8 = AnimStateTransitionNode()
   AnimStateTransitionNode_9 = AnimStateTransitionNode()
  ]
  links [
   {
    dst=Default
    rule=ItemUpperBody.AnimGraphNode_StateMachine_0.Block.AnimStateTransitionNode_9.Transition
    src=BackToDefault
    via=AnimStateTransitionNode_9
   }
   {
    dst=GuardIntro
    rule=ItemUpperBody.AnimGraphNode_StateMachine_0.Block.AnimStateTransitionNode_13.Transition
    src=Default
    via=AnimStateTransitionNode_13
   }
   {
    dst=BackToDefault
    rule=ItemUpperBody.AnimGraphNode_StateMachine_0.Block.AnimStateTransitionNode_6.Transition
    src=DeflectEntry1
    via=AnimStateTransitionNode_6
   }
   {
    dst=DeflectHold1
    rule=ItemUpperBody.AnimGraphNode_StateMachine_0.Block.AnimStateTransitionNode_1.AnimationTransitionGraph_0
    src=DeflectEntry1
    via=AnimStateTransitionNode_1
   }
   {
    dst=BackToDefault
    rule=ItemUpperBody.AnimGraphNode_StateMachine_0.Block.AnimStateTransitionNode_16.Transition
    src=DeflectExit1
    via=AnimStateTransitionNode_16
   }
   {
    dst=DeflectTransition1_2
    rule=ItemUpperBody.AnimGraphNode_StateMachine_0.Block.AnimStateTransitionNode_10.Transition
    src=DeflectExit1
    via=AnimStateTransitionNode_10
   }
   {
    dst=GuardLoop
    rule=ItemUpperBody.AnimGraphNode_StateMachine_0.Block.AnimStateTransitionNode_3.AnimationTransitionGraph_0
    src=DeflectExit1
    via=AnimStateTransitionNode_3
   }
   {
    dst=BackToDefault
    rule=ItemUpperBody.AnimGraphNode_StateMachine_0.Block.AnimStateTransitionNode_22.Transition
    src=DeflectExit2
    via=AnimStateTransitionNode_22
   }
   {
    dst=DeflectTransition2_1
    rule=ItemUpperBody.AnimGraphNode_StateMachine_0.Block.AnimStateTransitionNode_23.Transition
    src=DeflectExit2
    via=AnimStateTransitionNode_23
   }
   {
    dst=GuardLoop
    rule=ItemUpperBody.AnimGraphNode_StateMachine_0.Block.AnimStateTransitionNode_21.Transition
    src=DeflectExit2
    via=AnimStateTransitionNode_21
   }
   {
    dst=BackToDefault
    rule=ItemUpperBody.AnimGraphNode_StateMachine_0.Block.AnimStateTransitionNode_7.Transition
    src=DeflectHold1
    via=AnimStateTransitionNode_7
   }
   {
    dst=DeflectExit1
    rule=ItemUpperBody.AnimGraphNode_StateMachine_0.Block.AnimStateTransitionNode_2.AnimationTransitionGraph_0
    src=DeflectHold1
    via=AnimStateTransitionNode_2
   }
   {
    dst=DeflectTransition1_2
    rule=ItemUpperBody.AnimGraphNode_StateMachine_0.Block.AnimStateTransitionNode_12.Transition
    src=DeflectHold1
    via=AnimStateTransitionNode_12
   }
   {
    dst=BackToDefault
    rule=ItemUpperBody.AnimGraphNode_StateMachine_0.Block.AnimStateTransitionNode_11.Transition
    src=DeflectHold2
    via=AnimStateTransitionNode_11
   }
   {
    dst=DeflectExit2
    rule=ItemUpperBody.AnimGraphNode_StateMachine_0.Block.AnimStateTransitionNode_20.Transition
    src=DeflectHold2
    via=AnimStateTransitionNode_20
   }
   {
    dst=DeflectTransition2_1
    rule=ItemUpperBody.AnimGraphNode_StateMachine_0.Block.AnimStateTransitionNode_17.Transition
    src=DeflectHold2
    via=AnimStateTransitionNode_17
   }
   {
    dst=BackToDefault
    rule=ItemUpperBody.AnimGraphNode_StateMachine_0.Block.AnimStateTransitionNode_8.Transition
    src=DeflectTransition1_2
    via=AnimStateTransitionNode_8
   }
   {
    dst=DeflectHold2
    rule=ItemUpperBody.AnimGraphNode_StateMachine_0.Block.AnimStateTransitionNode_4.Transition
    src=DeflectTransition1_2
    via=AnimStateTransitionNode_4
   }
   {
    dst=BackToDefault
    rule=ItemUpperBody.AnimGraphNode_StateMachine_0.Block.AnimStateTransitionNode_18.Transition
    src=DeflectTransition2_1
    via=AnimStateTransitionNode_18
   }
   {
    dst=DeflectHold1
    rule=ItemUpperBody.AnimGraphNode_StateMachine_0.Block.AnimStateTransitionNode_19.Transition
    src=DeflectTransition2_1
    via=AnimStateTransitionNode_19
   }
   {
    dst=GuardLoop
    rule=ItemUpperBody.AnimGraphNode_StateMachine_0.Block.AnimStateTransitionNode_5.Transition
    src=GuardIntro
    via=AnimStateTransitionNode_5
   }
   {
    dst=DeflectEntry1
    rule=ItemUpperBody.AnimGraphNode_StateMachine_0.Block.AnimStateTransitionNode_0.AnimationTransitionGraph_0
    src=GuardLoop
    via=AnimStateTransitionNode_0
   }
   {
    dst=GuardToIdle
    rule=ItemUpperBody.AnimGraphNode_StateMachine_0.Block.AnimStateTransitionNode_14.Transition
    src=GuardLoop
    via=AnimStateTransitionNode_14
   }
   {
    dst=Default
    rule=ItemUpperBody.AnimGraphNode_StateMachine_0.Block.AnimStateTransitionNode_15.Transition
    src=GuardToIdle
    via=AnimStateTransitionNode_15
   }
  ]
 }
 BackToDefault {
  path=ItemUpperBody.AnimGraphNode_StateMachine_0.Block.AnimStateConduitNode_0.BackToDefault
  nodes [Result(bCanEnterTransition=NOT Boolean(A=Get bIsMeleeGuarding()))]
 }
 GuardLoop {
  path=ItemUpperBody.AnimGraphNode_StateMachine_0.Block.AnimStateNode_0.GuardLoop
  nodes [Output Animation Pose(Result=Sequence Player 'Melee_Pose_Block_CMM'() {sequence=/Game/Animation/Game/MainPlayer/Combat/Melee/Generic/SwordBlock/Melee_Pose_Block_CMM})]
 }
 DeflectEntry1 {
  path=ItemUpperBody.AnimGraphNode_StateMachine_0.Block.AnimStateNode_1.DeflectEntry1
  nodes [Output Animation Pose(Result=Sequence Player!(Sequence:Anim Sequence Base Object Reference=Get DeflectEntry1Anim() -> (DeflectEntry1Anim:Anim Sequence Object Reference)) -> (Pose) {bLoopAnimation=false})]
 }
 Default {
  path=ItemUpperBody.AnimGraphNode_StateMachine_0.Block.AnimStateNode_11.Default
  nodes [Output Animation Pose(Result=Use cached pose 'UpperBody_CachePose'())]
 }
 GuardToIdle {
  path=ItemUpperBody.AnimGraphNode_StateMachine_0.Block.AnimStateNode_12.GuardToIdle
  nodes [Output Animation Pose(Result=Sequence Player 'Melee_Outro_Block_CMM'() {bLoopAnimation=false, sequence=/Game/Animation/Game/MainPlayer/Combat/Melee/Generic/SwordBlock/Melee_Outro_Block_CMM})]
 }
 DeflectHold1 {
  path=ItemUpperBody.AnimGraphNode_StateMachine_0.Block.AnimStateNode_2.DeflectHold1
  nodes [Output Animation Pose(Result=Sequence Player(Sequence=Get DeflectHold1Anim() -> (DeflectHold1Anim:Anim Sequence Object Reference)) {bLoopAnimation=false})]
 }
 DeflectExit1 {
  path=ItemUpperBody.AnimGraphNode_StateMachine_0.Block.AnimStateNode_3.DeflectExit1
  nodes [Output Animation Pose(Result=Sequence Player(Sequence=Get DeflectExit1Anim() -> (DeflectExit1Anim:Anim Sequence Object Reference)) {bLoopAnimation=false})]
 }
 DeflectTransition1_2 {
  path=ItemUpperBody.AnimGraphNode_StateMachine_0.Block.AnimStateNode_4.DeflectTransition1_2
  nodes [Output Animation Pose(Result=Sequence Player(Sequence=Get DeflectTransition1_2Anim() -> (DeflectTransition1_2Anim:Anim Sequence Object Reference)) {bLoopAnimation=false})]
 }
 GuardIntro {
  path=ItemUpperBody.AnimGraphNode_StateMachine_0.Block.AnimStateNode_5.GuardIntro
  nodes [Output Animation Pose(Result=Sequence Player 'Melee_Into_Block_CMM'() {bLoopAnimation=false, sequence=/Game/Animation/Game/MainPlayer/Combat/Melee/Generic/SwordBlock/Melee_Into_Block_CMM})]
 }
 DeflectHold2 {
  path=ItemUpperBody.AnimGraphNode_StateMachine_0.Block.AnimStateNode_6.DeflectHold2
  nodes [Output Animation Pose(Result=Sequence Player(Sequence=Get DeflectHold2Anim() -> (DeflectHold2Anim:Anim Sequence Object Reference)) {bLoopAnimation=false})]
 }
 DeflectTransition2_1 {
  path=ItemUpperBody.AnimGraphNode_StateMachine_0.Block.AnimStateNode_7.DeflectTransition2_1
  nodes [Output Animation Pose(Result=Sequence Player(Sequence=Get DeflectTransition2_1Anim() -> (DeflectTransition2_1Anim:Anim Sequence Object Reference)) {bLoopAnimation=false})]
 }
 DeflectExit2 {
  path=ItemUpperBody.AnimGraphNode_StateMachine_0.Block.AnimStateNode_8.DeflectExit2
  nodes [Output Animation Pose(Result=Sequence Player(Sequence=Get DeflectExit2Anim() -> (DeflectExit2Anim:Anim Sequence Object Reference)) {bLoopAnimation=false})]
 }
 AnimationTransitionGraph_0 {
  path=ItemUpperBody.AnimGraphNode_StateMachine_0.Block.AnimStateTransitionNode_0.AnimationTransitionGraph_0
  nodes [Result(bCanEnterTransition=Get bIsMeleeDeflecting() -> (bIsMeleeDeflecting:bool))]
 }
 AnimationTransitionGraph_0 {path=ItemUpperBody.AnimGraphNode_StateMachine_0.Block.AnimStateTransitionNode_1.AnimationTransitionGraph_0, nodes [Result()]}
 Transition {
  path=ItemUpperBody.AnimGraphNode_StateMachine_0.Block.AnimStateTransitionNode_10.Transition
  nodes [Result(bCanEnterTransition=Get bIsMeleeDeflecting())]
 }
 Transition {
  path=ItemUpperBody.AnimGraphNode_StateMachine_0.Block.AnimStateTransitionNode_11.Transition
  nodes [Result(bCanEnterTransition=NOT Boolean(A=Get bIsMeleeGuarding()))]
 }
 Transition {
  path=ItemUpperBody.AnimGraphNode_StateMachine_0.Block.AnimStateTransitionNode_12.Transition
  nodes [Result(bCanEnterTransition=Get bIsMeleeDeflecting())]
 }
 Transition {
  path=ItemUpperBody.AnimGraphNode_StateMachine_0.Block.AnimStateTransitionNode_13.Transition
  nodes [Result(bCanEnterTransition=Get bIsMeleeGuarding())]
 }
 Transition {
  path=ItemUpperBody.AnimGraphNode_StateMachine_0.Block.AnimStateTransitionNode_14.Transition
  nodes [Result(bCanEnterTransition=NOT Boolean(A=Get bIsMeleeGuarding()))]
 }
 Transition {path=ItemUpperBody.AnimGraphNode_StateMachine_0.Block.AnimStateTransitionNode_15.Transition, nodes [Result()]}
 Transition {
  path=ItemUpperBody.AnimGraphNode_StateMachine_0.Block.AnimStateTransitionNode_16.Transition
  nodes [Result(bCanEnterTransition=NOT Boolean(A=Get bIsMeleeGuarding()))]
 }
 Transition {
  path=ItemUpperBody.AnimGraphNode_StateMachine_0.Block.AnimStateTransitionNode_17.Transition
  nodes [Result(bCanEnterTransition=Get bIsMeleeDeflecting())]
 }
 Transition {
  path=ItemUpperBody.AnimGraphNode_StateMachine_0.Block.AnimStateTransitionNode_18.Transition
  nodes [Result(bCanEnterTransition=NOT Boolean(A=Get bIsMeleeGuarding()))]
 }
 Transition {path=ItemUpperBody.AnimGraphNode_StateMachine_0.Block.AnimStateTransitionNode_19.Transition, nodes [Result()]}
 AnimationTransitionGraph_0 {path=ItemUpperBody.AnimGraphNode_StateMachine_0.Block.AnimStateTransitionNode_2.AnimationTransitionGraph_0, nodes [Result()]}
 Transition {path=ItemUpperBody.AnimGraphNode_StateMachine_0.Block.AnimStateTransitionNode_20.Transition, nodes [Result()]}
 Transition {path=ItemUpperBody.AnimGraphNode_StateMachine_0.Block.AnimStateTransitionNode_21.Transition, nodes [Result()]}
 Transition {
  path=ItemUpperBody.AnimGraphNode_StateMachine_0.Block.AnimStateTransitionNode_22.Transition
  nodes [Result(bCanEnterTransition=NOT Boolean(A=Get bIsMeleeGuarding()))]
 }
 Transition {
  path=ItemUpperBody.AnimGraphNode_StateMachine_0.Block.AnimStateTransitionNode_23.Transition
  nodes [Result(bCanEnterTransition=Get bIsMeleeDeflecting())]
 }
 AnimationTransitionGraph_0 {path=ItemUpperBody.AnimGraphNode_StateMachine_0.Block.AnimStateTransitionNode_3.AnimationTransitionGraph_0, nodes [Result()]}
 Transition {path=ItemUpperBody.AnimGraphNode_StateMachine_0.Block.AnimStateTransitionNode_4.Transition, nodes [Result()]}
 Transition {path=ItemUpperBody.AnimGraphNode_StateMachine_0.Block.AnimStateTransitionNode_5.Transition, nodes [Result()]}
 Transition {
  path=ItemUpperBody.AnimGraphNode_StateMachine_0.Block.AnimStateTransitionNode_6.Transition
  nodes [Result(bCanEnterTransition=NOT Boolean(A=Get bIsMeleeGuarding()))]
 }
 Transition {
  path=ItemUpperBody.AnimGraphNode_StateMachine_0.Block.AnimStateTransitionNode_7.Transition
  nodes [Result(bCanEnterTransition=NOT Boolean(A=Get bIsMeleeGuarding()))]
 }
 Transition {
  path=ItemUpperBody.AnimGraphNode_StateMachine_0.Block.AnimStateTransitionNode_8.Transition
  nodes [Result(bCanEnterTransition=NOT Boolean(A=Get bIsMeleeGuarding()))]
 }
 Transition {
  path=ItemUpperBody.AnimGraphNode_StateMachine_0.Block.AnimStateTransitionNode_9.Transition
  nodes [Result(bCanEnterTransition=NOT Boolean(A=Get bIsMeleeGuarding()))]
 }
 ItemZipLineStartAdditive {nodes [Output Pose(Result=Blend Poses by bool(
     BlendPose_0="Sequence Player (additive) 'ZipLine_Entry_ARDS_Additive'"() {bLoopAnimation=false, sequence=/Game/Animation/Game/MainPlayer/Combat/Gadgets/Medium/Male/ZipLine/ZipLine_Entry_ARDS_Additive}
     BlendPose_1="Sequence Player (additive) 'ZipLine_Entry_NoWep_Additive'"() {bLoopAnimation=false, sequence=/Game/Animation/Game/MainPlayer/Combat/Gadgets/Medium/Male/ZipLine/ZipLine_Entry_NoWep_Additive}
     bActiveValue=Get bIsMeleeGuarding()
    ) {blendType=Linear}) {layerGroup=Item Override}]}
]
variables []
