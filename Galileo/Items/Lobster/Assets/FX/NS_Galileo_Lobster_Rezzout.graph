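# Niagara system for the rez-out effect of the Galileo Lobster saber blade, tinted by user color parameters.
# The shared User.CurveSaberAll curve drives mesh length, glow sprites and a moving flare, keeping them in step.
# Timed flares, a base glow and lights add flashes and flicker.
# Enable Blade Arc gates jittered Hermite-spline ribbon arcs and curl-noise GPU sparks, while LoopDelayAll offsets the arcs and one flare.
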NiagaraSystem /Galileo/Items/Lobster/Assets/FX/NS_Galileo_Lobster_Rezzout
dependencies {usedDataInterfaces [], usedDynamicInputs [], usedModules [&SystemState]}
emitters {
 Blade_Arcs {
  emitterUpdateScript [
   &EmitterState(
    Delay First Loop Only=true
    Inactive Response=::NewEnumerator1
    Life Cycle Mode=::NewEnumerator1
    Loop Delay=&Add_Float(A=$User.LoopDelayAll, B=0.15)
    Loop Duration=1
    UseLoopDelay=true
   )
   /SetVariables(Emitter.SpawnCount:NiagaraInt32=&RandomRangeInt(Minimum=5, Maximum=11)) {isSetParameters=true}
   /SetVariables(Emitter.Lifetime:NiagaraFloat=&RandomRangeFloat(Minimum=1, Maximum=1)) {isSetParameters=true}
   &SpawnBurst_Instantaneous(
    Spawn Count=$Emitter.SpawnCount
    Spawn Probability=&MakeCustomFloatFromBool(Bool=$User.Enable Blade Arc, True Float=1, False Float=0)
    Use Spawn Probability=true
    UseLoopCountLimit=false
   )
  ]
  particleSpawnScript [
   &PartitionParticles(Number Of Partitions=$Emitter.SpawnCount)
   /SetVariables(Particles.Normalized Spawn Group:NiagaraFloat=&Multiply_Float(
     A=&Divide_Float(
      A=1
      B=&Subtract_Float(A=&MakeFloatFromInt(Int=$Emitter.SpawnCount), B=1)
     )
     B=&MakeFloatFromInt(Int=$Particles.PartitionParticles.Partition)
    )) {isSetParameters=true}
   /SetVariables(Particles.RibbonLinkOrder:NiagaraFloat=&OneMinusFloat(Float=$Particles.Normalized Spawn Group)) {isSetParameters=true}
   &InitializeRibbon(
    Lifetime=$Emitter.Lifetime
    Ribbon Width=&FloatFromCurve(
     FloatCurve=DataInterface() {propertyValues {Curve {keys [<at 0 0.75>,<at 0.25 1>,<at 0.66 1>,<at 1 0.33>]}}}
     CurveIndex=$Particles.Normalized Spawn Group
     Scale Curve=&RandomRangeFloat(Minimum=2, Maximum=3)
    )
    UsePositionOffset=false
    Write Material Random=true
   )
   /SetVariables(Particles.SpriteSize=&Vector2DFromFloat(Value=&FloatFromCurve(
      FloatCurve=DataInterface() {propertyValues {Curve {keys [{tangentMode=RCTM_User, time=0, value=0.5}, {tangentMode=RCTM_User, time=1, value=1}]}}}
      CurveIndex=$Particles.Normalized Spawn Group
      Scale Curve=2
     ))) {isSetParameters=true}
  ]
  particleUpdateScript [
   &ParticleState()
   /Game/Effects/Niagara/Modules/Curves/FN_HermiteSpline(
    Direction End:Vector3f=&Multiply_VectorByFloat(
     Vector=&MakeVector(
      X=&RandomRangeFloat(Minimum=1, Maximum=-1)
      Y=&RandomRangeFloat(Minimum=1, Maximum=-1)
      Z=&RandomRangeFloat(Minimum=-1, Maximum=-1)
     )
     Float=&Waveform(
      Global Amplitude Scale=&RandomRangeFloat(Minimum=10, Maximum=20)
      "Waveform [1]"=::NewEnumerator0
      "[1] Frequency"=2
      "[1] Amplitude Min/Max"=<0.8 1>
     )
    )
    Direction Start:Vector3f=&Multiply_VectorByFloat(
     Vector=&MakeVector(
      X=&RandomRangeFloat(Minimum=-1, Maximum=1)
      Y=&RandomRangeFloat(Minimum=-1, Maximum=1)
      Z=&RandomRangeFloat(Minimum=0.1, Maximum=0.5)
     )
     Float=&Waveform(
      Global Amplitude Scale=&RandomRangeFloat(Minimum=30, Maximum=40)
      "Waveform [1]"=::NewEnumerator12
      "[1] Frequency"=1.5
      "[1] Amplitude Min/Max"=<0.8 1>
     )
    )
    Position Blend:NiagaraFloat=$Particles.Normalized Spawn Group
    Position End:Vector3f=<0 0 107>
    Position Start:Vector3f=<0 0 12>
    Write Particles Position=true
   )
   &JitterPosition(
    Jitter Amount:NiagaraFloat=&FloatFromCurve(
     FloatCurve=DataInterface() {propertyValues {Curve {keys [<at 0 0>,<at 0.25 1>,<at 0.66 1>,<at 1 0>]}}}
     CurveIndex=$Particles.Normalized Spawn Group
     Scale Curve=&RandomRangeFloat(Minimum=1, Maximum=2.5)
    )
    Jitter Offset:Vector3f=&Multiply_Vector(A=&RandomVector(Vector Scale=1, Evaluation Type=::NewEnumerator1), B=<1 1 0.01>)
   )
  ]
  rendererClasses [&NiagaraRibbonRendererProperties, &NiagaraSpriteRendererProperties]
  renderers [
   {rendererClass=&NiagaraRibbonRendererProperties}
   {rendererClass=&NiagaraSpriteRendererProperties, rendererIndex=1}
  ]
 }
 Blade_Flare_Activate {
  emitterUpdateScript [
   &EmitterState(Inactive Response=::NewEnumerator1, Life Cycle Mode=::NewEnumerator1, Loop Delay=$User.LoopDelayAll, Loop Duration=1, UseLoopDelay=true)
   /SetVariables(Emitter.WideFlare:NiagaraInt32=&MakeCustomIntFromBool(Bool=$User.Enable Blade Arc, True INT=1, False INT=0)) {isSetParameters=true, module=SetVariables_#1}
   &SpawnBurst_Instantaneous(Use Spawn Probability=false, UseLoopCountLimit=true)
  ]
  particleSpawnScript [
   &InitializeParticle(
    Position Offset=<0 0 16.5>
    Sprite Size Max=<125 75>
    Sprite Size Min=<112.5 67.5>
    Sprite Size Mode=::NewEnumerator2
    UsePositionOffset=true
    Write Material Random=true
   )
   &SpriteFacingAndAlignment(Sprite Alignment=<0 0 1>, Sprite Facing=<0 0 1>, WriteAlignment=true, WriteFacing=false)
   &Color(Color=$User.Color_Glow)
   &CameraOffset(Camera Offset Amount=50)
  ]
  particleUpdateScript [
   /SetVariables(Particles.CurveSaber=DataInterface() {propertyValues {Curve {keys [{tangentMode=RCTM_User, time=0, value=110}, {tangentMode=RCTM_User, time=1, value=0}]}}}) {isSetParameters=true}
   &ParticleState()
   &ScaleColor(
    Scale Alpha=&FloatFromCurve(FloatCurve=DataInterface() {propertyValues {Curve {keys [{arriveTangent=6.6296, time=0.9, value=0}, <at 0.95 1>, <at 1 0>]}}})
    Scale RGB=<3 3 3>
    ScaleA=true
    ScaleRGB=true
   )
   &ScaleSpriteSize(Uniform Curve Sprite Scale=DataInterface() {propertyValues {Curve {keys [
        {arriveTangent=4.1355, interpMode=RCIM_Cubic, leaveTangent=4.1355, tangentMode=RCTM_User, time=0.8, value=0.25}
        {interpMode=RCIM_Cubic, tangentMode=RCTM_User, time=0.84, value=0.9591}
        {tangentMode=RCTM_User, time=1, value=0.25}
       ]}}})
   /SetVariables(
    Particles.Color Wide:LinearColor=&MultiplyLinearColor(Linear Color=$Particles.Color, Scale Factor=<4 4 4 1>)
    Particles.SpriteSize_Wide:Vector2f=&Multiply_Vector2D(A=$Particles.SpriteSize, B=<1.5 0.8>)
   ) {isSetParameters=true, module=SetVariables_#2}
   /SetVariables(Particles.Position_Light:NiagaraPosition=&Add_Vector(
     A=Linked() {linkedVariable {name=Particles.Initial.Position, type=NiagaraPosition}}
     B=&Multiply_VectorByFloat(Vector=<0 0 110>, Float=&FloatFromCurve(FloatCurve=$User.CurveSaberAll))
    )) {isSetParameters=true, module=SetVariables_#3}
   /SetVariables(Particles.Position=&Add_Vector(
     A=&MakeVector(
      X=0
      Y=0
      Z=&Multiply_Float(A=&FloatFromCurve(FloatCurve=$User.CurveSaberAll, Scale Curve=110), B=1)
     )
     B=Linked() {linkedVariable {name=Particles.Initial.Position, type=NiagaraPosition}}
    )) {isSetParameters=true}
   &FN_Light_Attributes(
    Color=$Particles.Color
    Create Color Attribute=true
    Create Exponent Attribute=true
    Create Radius Attribute=true
    Exponent=3
    Radius=&Multiply_Float(A=&MakeFloatFromVector2d(Vector 2D=$Particles.SpriteSize), B=3)
    Scale Color=2
   )
  ]
  rendererClasses [&NiagaraSpriteRendererProperties, &NiagaraLightRendererProperties]
  renderers [
   {rendererClass=&NiagaraSpriteRendererProperties}
   {rendererClass=&NiagaraSpriteRendererProperties, rendererIndex=1}
   {rendererClass=&NiagaraLightRendererProperties, rendererIndex=2}
  ]
 }
 Blade_Flare_Activate001 {
  emitterUpdateScript [
   &EmitterState(Inactive Response=::NewEnumerator1, Life Cycle Mode=::NewEnumerator1, Loop Duration=1, UseLoopDelay=false)
   /SetVariables(Emitter.WideFlare:NiagaraInt32=&MakeCustomIntFromBool(Bool=$User.Enable Blade Arc, True INT=1, False INT=0)) {isSetParameters=true, module=SetVariables_#1}
   &SpawnBurst_Instantaneous(Use Spawn Probability=false, UseLoopCountLimit=true)
  ]
  particleSpawnScript [
   &InitializeParticle(
    Position Offset=<0 0 16.5>
    Sprite Size Max=<125 75>
    Sprite Size Min=<112.5 67.5>
    Sprite Size Mode=::NewEnumerator2
    UsePositionOffset=true
    Write Material Random=true
   )
   &SpriteFacingAndAlignment(Sprite Alignment=<0 0 1>, Sprite Facing=<0 0 1>, WriteAlignment=true, WriteFacing=false)
   &Color(Color=$User.Color_Glow)
   &CameraOffset(Camera Offset Amount=50)
  ]
  particleUpdateScript [
   &ParticleState()
   &ScaleColor(
    Scale Alpha=&FloatFromCurve(FloatCurve=DataInterface() {propertyValues {Curve {keys [
         {arriveTangent=1.1046, interpMode=RCIM_Cubic, leaveTangent=1.1046, tangentMode=RCTM_User, time=0, value=0}
         {interpMode=RCIM_Cubic, tangentMode=RCTM_User, time=0.8, value=1}
         {tangentMode=RCTM_User, time=1, value=0}
        ]}}})
    ScaleA=true
    ScaleRGB=true
   )
   &ScaleSpriteSize(Uniform Curve Sprite Scale=DataInterface() {propertyValues {Curve {keys [
        {arriveTangent=4.1355, interpMode=RCIM_Cubic, leaveTangent=4.1355, tangentMode=RCTM_User, time=0, value=0.25}
        {interpMode=RCIM_Cubic, tangentMode=RCTM_User, time=0.2, value=1}
        {tangentMode=RCTM_User, time=1, value=0.25}
       ]}}})
   /SetVariables(
    Particles.Color Wide:LinearColor=&MultiplyLinearColor(Linear Color=$Particles.Color, Scale Factor=<4 4 4 1>)
    Particles.SpriteSize_Wide:Vector2f=&Multiply_Vector2D(A=$Particles.SpriteSize, B=<1.5 0.8>)
   ) {isSetParameters=true, module=SetVariables_#2}
   /SetVariables(Particles.Position_Light:NiagaraPosition=&Add_Vector(
     A=Linked() {linkedVariable {name=Particles.Initial.Position, type=NiagaraPosition}}
     B=&Multiply_VectorByFloat(
      Vector=<0 0 110>
      Float=&FloatFromCurve(FloatCurve=DataInterface() {propertyValues {Curve {keys [{interpMode=RCIM_Cubic, tangentMode=RCTM_User, time=0, value=0}, {interpMode=RCIM_Cubic, tangentMode=RCTM_User, time=1, value=1}]}}})
     )
    )) {isSetParameters=true, module=SetVariables_#3}
   &FN_Light_Attributes(
    Color=$Particles.Color
    Create Color Attribute=true
    Create Exponent Attribute=true
    Create Radius Attribute=true
    Exponent=3
    Radius=&Multiply_Float(A=&MakeFloatFromVector2d(Vector 2D=$Particles.SpriteSize), B=3)
    Scale Color=2
   )
  ]
  rendererClasses [&NiagaraSpriteRendererProperties, &NiagaraLightRendererProperties]
  renderers [
   {rendererClass=&NiagaraSpriteRendererProperties}
   {rendererClass=&NiagaraSpriteRendererProperties, rendererIndex=1}
   {rendererClass=&NiagaraLightRendererProperties, rendererIndex=2}
  ]
 }
 Blade_Glow {
  emitterUpdateScript [
   &EmitterState(Inactive Response=::NewEnumerator1, Life Cycle Mode=::NewEnumerator1, Loop Duration=1, UseLoopDelay=false)
   /SetVariables(Emitter.WideFlare:NiagaraInt32=&MakeCustomIntFromBool(Bool=$User.Enable Blade Arc, True INT=1, False INT=1)) {isSetParameters=true, module=SetVariables_#4}
   &SpawnBurst_Instantaneous(Use Spawn Probability=false, UseLoopCountLimit=true)
  ]
  particleSpawnScript [
   &InitializeParticle(Position Offset=<0 0 64>, Sprite Size=<20 110>, Sprite Size Mode=::NewEnumerator4, UsePositionOffset=true, Write Material Random=true)
   &SpriteFacingAndAlignment(Sprite Alignment=<0 0 1>, Sprite Facing=<0 0 1>, WriteAlignment=true, WriteFacing=false)
   &Color(Color=$User.Color_Glow, Scale Alpha=0.3)
   &DynamicMaterialParameters(
    Index 0 Param 1=1.2
    Index 0 Param 2=20
    Index 0 Param 3=0.5
    Module.Param0WriteEnabled=true
    Module.Param1WriteEnabled=true
    Module.Param2WriteEnabled=true
    Module.Param3WriteEnabled=true
   )
  ]
  particleUpdateScript [
   &ParticleState()
   &ScaleColor(
    Scale Alpha=&FloatFromCurve(
     FloatCurve=DataInterface() {propertyValues {Curve {keys [{interpMode=RCIM_Cubic, tangentMode=RCTM_User, time=0, value=0}, {interpMode=RCIM_Cubic, tangentMode=RCTM_User, time=0.5, value=1}]}}}
     CurveIndex=$Emitter.Age
     Scale Curve=&Waveform("Waveform [1]"=::NewEnumerator10, "[1] Frequency"=10, "[1] Amplitude Min/Max"=<0.75 1>)
    )
    ScaleA=true
    ScaleRGB=true
   )
   &FN_Light_Attributes(Color=$Particles.Color, Create Color Attribute=true, Create Radius Attribute=true, Radius=160, Scale Color=8)
   /SetVariables(Particles.SpriteSizeBlend:Vector2f=&Multiply_Vector2D(A=$Particles.SpriteSize, B=&MakeVector2D(X=0.75, Y=0.9))) {isSetParameters=true, module=SetVariables_#5}
   /SetVariables(Particles.SpriteSize=&MakeVector2D(X=20, Y=&FloatFromCurve(FloatCurve=$User.CurveSaberAll, Scale Curve=110))) {isSetParameters=true, module=SetVariables_#6}
   /SetVariables(Particles.Position=&Multiply_Vector(
     A=Linked() {linkedVariable {name=Particles.Initial.Position, type=NiagaraPosition}}
     B=/Niagara/DynamicInputs/Lerp/Lerp_Position(
      A:NiagaraPosition=<0 0 0.4>
      B:NiagaraPosition=<0 0 -0.2>
      Alpha:NiagaraFloat=&Multiply_Float(A=&FloatFromCurve(FloatCurve=$User.CurveSaberAll), B=-1)
     )
    )) {isSetParameters=true, module=SetVariables_#7}
  ]
  rendererClasses [&NiagaraSpriteRendererProperties, &NiagaraLightRendererProperties]
  renderers [
   {rendererClass=&NiagaraSpriteRendererProperties}
   {rendererClass=&NiagaraLightRendererProperties, rendererIndex=1}
   {rendererClass=&NiagaraSpriteRendererProperties, rendererIndex=2}
  ]
 }
 Blade_Glow001 {
  bEnabled=false
  emitterUpdateScript [
   &EmitterState(Inactive Response=::NewEnumerator1, Life Cycle Mode=::NewEnumerator1, Loop Duration=1, UseLoopDelay=false)
   /SetVariables(Emitter.WideFlare:NiagaraInt32=&MakeCustomIntFromBool(Bool=$User.Enable Blade Arc, True INT=1, False INT=1)) {isSetParameters=true, module=SetVariables_#4}
   &SpawnBurst_Instantaneous(Use Spawn Probability=false, UseLoopCountLimit=true)
  ]
  particleSpawnScript [
   &InitializeParticle(Position Offset=<0 0 64>, Sprite Size=<20 110>, Sprite Size Mode=::NewEnumerator4, UsePositionOffset=true, Write Material Random=true)
   &SpriteFacingAndAlignment(Sprite Alignment=<0 0 1>, Sprite Facing=<0 0 1>, WriteAlignment=true, WriteFacing=false)
   &Color(Color=<rgba 0 0.1918 1 1>, Scale Alpha=0.3)
   &DynamicMaterialParameters(
    Index 0 Param 1=1.2
    Index 0 Param 2=20
    Index 0 Param 3=0.5
    Module.Param0WriteEnabled=true
    Module.Param1WriteEnabled=true
    Module.Param2WriteEnabled=true
    Module.Param3WriteEnabled=true
   )
  ]
  particleUpdateScript [
   &ParticleState(Kill Particles When Lifetime Has Elapsed=false, Loop Particles Lifetime=true)
   &ScaleColor(
    Scale Alpha=&FloatFromCurve(
     FloatCurve=DataInterface() {propertyValues {Curve {keys [{interpMode=RCIM_Cubic, tangentMode=RCTM_User, time=0, value=0}, {interpMode=RCIM_Cubic, tangentMode=RCTM_User, time=0.5, value=1}]}}}
     CurveIndex=$Emitter.Age
     Scale Curve=&Waveform("Waveform [1]"=::NewEnumerator10, "[1] Frequency"=10, "[1] Amplitude Min/Max"=<0.75 1>)
    )
    ScaleA=true
    ScaleRGB=true
   )
   &FN_Light_Attributes(Color=$Particles.Color, Create Color Attribute=true, Create Radius Attribute=true, Radius=160, Scale Color=8)
   /SetVariables(Particles.SpriteSizeBlend:Vector2f=&Multiply_Vector2D(A=$Particles.SpriteSize, B=<0.75 0.9>)) {isSetParameters=true, module=SetVariables_#5}
  ]
  rendererClasses [&NiagaraSpriteRendererProperties, &NiagaraLightRendererProperties]
  renderers [
   {rendererClass=&NiagaraSpriteRendererProperties}
   {rendererClass=&NiagaraLightRendererProperties, rendererIndex=1}
   {rendererClass=&NiagaraSpriteRendererProperties, rendererIndex=2}
  ]
 }
 Blade_Glow_Activate {
  emitterUpdateScript [
   &EmitterState(Inactive Response=::NewEnumerator1, Life Cycle Mode=::NewEnumerator1, Loop Duration=1, UseLoopDelay=false)
   &SpawnBurst_Instantaneous(Spawn Time=0.1, Use Spawn Probability=false, UseLoopCountLimit=true)
  ]
  particleSpawnScript [
   &InitializeParticle(
    Lifetime=0.8
    Position Offset=<0 0 75>
    Sprite Size=<20 100>
    Sprite Size Mode=::NewEnumerator4
    UsePositionOffset=true
    Write Material Random=true
   )
   &SpriteFacingAndAlignment(Sprite Alignment=<0 0 1>, Sprite Facing=<0 0 1>, WriteAlignment=true, WriteFacing=false)
   &Color(Color=$User.Color_Glow, Scale Color=<1.5 1.5 1.5>)
   &DynamicMaterialParameters(
    Index 0 Param 1=1
    Index 0 Param 2=20
    Index 0 Param 3=0.4
    Module.Param0WriteEnabled=true
    Module.Param1WriteEnabled=true
    Module.Param2WriteEnabled=true
    Module.Param3WriteEnabled=true
   )
  ]
  particleUpdateScript [
   &ParticleState()
   &ScaleSpriteSize(
    Scale Factor=&MakeVector2D(
     X=&FloatFromCurve(FloatCurve=DataInterface() {propertyValues {Curve {keys [
          {interpMode=RCIM_Cubic, tangentMode=RCTM_User, time=0, value=0}
          {interpMode=RCIM_Cubic, tangentMode=RCTM_User, time=0.1, value=1}
          {tangentMode=RCTM_User, time=1, value=0.2}
         ]}}})
     Y=&FloatFromCurve(FloatCurve=$User.CurveSaberAll)
    )
    Scale Sprite Size Mode=::NewEnumerator2
   )
   /SetVariables(Particles.Position=&Add_Vector(
     A=Linked() {linkedVariable {name=Particles.Initial.Position, type=NiagaraPosition}}
     B=&Multiply_VectorByFloat(Vector=<0 0 -10>, Float=&FloatFromCurve(FloatCurve=$User.CurveSaberAll))
    )) {isSetParameters=true}
   &ScaleColor(
    Scale Alpha=&FloatFromCurve(
     FloatCurve=DataInterface() {propertyValues {Curve {keys [
         {arriveTangent=7.1958, interpMode=RCIM_Cubic, leaveTangent=7.1958, tangentMode=RCTM_User, time=0, value=0}
         {interpMode=RCIM_Cubic, tangentMode=RCTM_User, time=0.25, value=1}
         {interpMode=RCIM_Cubic, leaveTangent=-6.0579, time=1, value=0}
        ]}}}
     Scale Curve=5
    )
    ScaleA=true
    ScaleRGB=true
   )
  ]
  rendererClasses [&NiagaraSpriteRendererProperties]
  renderers [{rendererClass=&NiagaraSpriteRendererProperties}]
 }
 Glow_Circle_Base {
  emitterUpdateScript [
   &EmitterState(Inactive Response=::NewEnumerator1, Life Cycle Mode=::NewEnumerator1, Loop Duration=1, UseLoopDelay=false)
   /SetVariables(Emitter.WideFlare:NiagaraInt32=&MakeCustomIntFromBool(Bool=$User.Enable Blade Arc, True INT=1, False INT=1)) {isSetParameters=true, module=SetVariables_#4}
   &SpawnBurst_Instantaneous(Use Spawn Probability=false, UseLoopCountLimit=true)
  ]
  particleSpawnScript [
   &InitializeParticle(
    Lifetime=1.05
    Position Offset=<0 0 64>
    Sprite Size=<20 110>
    Sprite Size Mode=::NewEnumerator4
    UsePositionOffset=true
    Write Material Random=true
   )
   &SpriteFacingAndAlignment(Sprite Alignment=<0 0 1>, Sprite Facing=<0 0 1>, WriteAlignment=true, WriteFacing=false)
   &Color(
    Color=&HueShiftLinearColor(Hue=0, Saturation=0.9, Value=15, Linear Color=$User.Color_Glow)
    Scale Alpha=0.06
   )
   &DynamicMaterialParameters(
    Index 0 Param 1=1.2
    Index 0 Param 2=20
    Index 0 Param 3=0.5
    Module.Param0WriteEnabled=true
    Module.Param1WriteEnabled=true
    Module.Param2WriteEnabled=true
    Module.Param3WriteEnabled=true
   )
  ]
  particleUpdateScript [
   &ParticleState()
   &ScaleColor(
    Scale Alpha=&FloatFromCurve(
     FloatCurve=DataInterface() {propertyValues {Curve {keys [{interpMode=RCIM_Cubic, tangentMode=RCTM_User, time=0, value=0}, {interpMode=RCIM_Cubic, tangentMode=RCTM_User, time=0.5, value=1}]}}}
     CurveIndex=$Emitter.Age
     Scale Curve=&Waveform("Waveform [1]"=::NewEnumerator10, "[1] Frequency"=10, "[1] Amplitude Min/Max"=<0.75 1>)
    )
    ScaleA=true
    ScaleRGB=true
   )
   &FN_Light_Attributes(Color=$Particles.Color, Create Color Attribute=true, Create Radius Attribute=true, Radius=160, Scale Color=8)
   /SetVariables(
    Particles.ColorAlt:LinearColor=<rgba 10 3 3 1>
    Particles.Position_Light:NiagaraPosition=&Multiply_Vector(
     A=<1 1 300>
     B=&MakeVector(X=0, Y=0, Z=&FloatFromCurve(FloatCurve=$User.CurveSaberAll, Scale Curve=0.1))
    )
    Particles.SpriteSizeBlend:Vector2f=&Multiply_Vector2D(A=$Particles.SpriteSize, B=&MakeVector2D(X=0.2, Y=1))
   ) {isSetParameters=true, module=SetVariables_#5}
   /SetVariables(Particles.SpriteSize=&MakeVector2D(
     X=&FloatFromCurve(
      FloatCurve=DataInterface() {propertyValues {Curve {keys [
          {arriveTangent=0.1558, interpMode=RCIM_Cubic, time=0.6, value=0}
          {arriveTangent=2.3333, interpMode=RCIM_Cubic, leaveTangent=2.3333, time=0.7, value=0.55}
          <at 0.9 0.7>
          {tangentMode=RCTM_User, time=1, value=0}
         ]}}}
      Scale Curve=50
     )
     Y=&FloatFromCurve(
      FloatCurve=DataInterface() {propertyValues {Curve {keys [
          {arriveTangent=-2.5, interpMode=RCIM_Cubic, time=0.6, value=0}
          {arriveTangent=3.7212, interpMode=RCIM_Cubic, leaveTangent=3.7212, time=0.6872, value=0.4946}
          <at 0.7532 0.5701>
          {tangentMode=RCTM_User, time=1, value=0}
         ]}}}
      Scale Curve=80
     )
    )) {isSetParameters=true, module=SetVariables_#6}
   /SetVariables(Particles.Position=&Multiply_Vector(A=Linked() {linkedVariable {name=Particles.Initial.Position, type=NiagaraPosition}}, B=<1 1 0.35>)) {isSetParameters=true, module=SetVariables_#7}
   &CameraOffset(Camera Offset Amount=5) {enabled=false}
  ]
  rendererClasses [&NiagaraSpriteRendererProperties, &NiagaraLightRendererProperties]
  renderers [
   {rendererClass=&NiagaraSpriteRendererProperties}
   {rendererClass=&NiagaraLightRendererProperties, rendererIndex=1}
   {rendererClass=&NiagaraSpriteRendererProperties, rendererIndex=2}
  ]
 }
 Saber {
  emitterUpdateScript [
   &EmitterState(Life Cycle Mode=::NewEnumerator1, Loop Behavior=::NewEnumerator1, Loop Duration=1, UseLoopDelay=false)
   &SpawnBurst_Instantaneous(Use Spawn Probability=false, UseLoopCountLimit=false)
  ]
  particleSpawnScript [&InitializeParticle(Mesh Scale Mode=::NewEnumerator3, Position=<0 0 15>, Position Mode=::NewEnumerator1, UsePositionOffset=false, Write Material Random=true)]
  particleUpdateScript [
   &ParticleState()
   &ScaleMeshSize(Scale Factor=&MakeVector(X=1, Y=1, Z=&FloatFromCurve(FloatCurve=$User.CurveSaberAll)))
   &DynamicMaterialParameters(Module.Param0WriteEnabled=true, Module.Param1WriteEnabled=true, Module.Param2WriteEnabled=true, Module.Param3WriteEnabled=true)
  ]
  rendererClasses [&NiagaraSpriteRendererProperties, &NiagaraMeshRendererProperties]
  renderers [
   {rendererClass=&NiagaraSpriteRendererProperties}
   {rendererClass=&NiagaraMeshRendererProperties, rendererIndex=1}
  ]
 }
 Sparks {
  simTarget=GPUComputeSim
  emitterUpdateScript [
   &EmitterState(Life Cycle Mode=::NewEnumerator1, Loop Delay=0.15, Loop Duration=1, UseLoopDelay=true)
   &SpawnRate(
    Spawn Probability=&MakeCustomFloatFromBool(Bool=$User.Enable Blade Arc, True Float=1, False Float=0)
    SpawnRate=15
    Use Spawn Probability=true
   )
  ]
  particleSpawnScript [
   &InitializeParticle(
    Color=$User.Color_Blade_A
    Lifetime Max=0.3
    Lifetime Min=0.1
    Lifetime Mode=::NewEnumerator1
    Position Offset=&Lerp_Vector(A=<0 0 12>, B=<0 0 100>, Alpha=&RandomRangeFloat(Minimum=0, Maximum=1))
    Position Offset Coordinate Space=::Local
    Sprite Size Mode=::NewEnumerator1
    Uniform Sprite Size Max=1
    Uniform Sprite Size Min=0.75
    UsePositionOffset=true
    Write Material Random=true
   )
   &AddVelocity(
    Cone Angle=&RandomRangeFloat(Minimum=0, Maximum=0.05)
    Cone Angle Mode:ENiagara_AngleInput=::NewEnumerator1
    Cone Axis=<0 0 1>
    Module.UseConeFalloff=false
    Module.UseVelDistribution=false
    Override Seed=false
    Velocity Mode=::NewEnumerator2
    Velocity Speed=&RandomRangeFloat(Minimum=50, Maximum=150)
   )
   &InheritVelocity(Inherited Velocity Amount Scale=<0.5 0.5 0.5>, Speed Limit=true, UseSpeedThreshold=false)
  ]
  particleUpdateScript [
   &ParticleState()
   &Drag(Drag=10, Use Linear Drag=true, Use Rotational Drag=false)
   &CurlNoiseForce(Noise Frequency=5000, Noise Strength=2000, Pan Noise=false)
   &SolveForcesAndVelocity(Clamp Velocity=false, Limit Acceleration=false)
   &ScaleSpriteSizeBySpeed(Max Scale Factor=<0.75 8>, Velocity Threshold=300)
   &ScaleColor(
    Scale Alpha=&FloatFromCurve(FloatCurve=DataInterface() {propertyValues {Curve {keys [{interpMode=RCIM_Cubic, tangentMode=RCTM_User, time=0.5, value=1}, {interpMode=RCIM_Cubic, tangentMode=RCTM_User, time=1, value=0}]}}})
    Scale RGB=&Lerp_Vector(
     A=<1 1 1>
     B=<4 4 4>
     Alpha=&FloatFromCurve(FloatCurve=DataInterface() {propertyValues {Curve {keys [{interpMode=RCIM_Cubic, tangentMode=RCTM_User, time=0, value=1}, {interpMode=RCIM_Cubic, tangentMode=RCTM_User, time=0.66, value=0}]}}})
    )
    ScaleA=true
    ScaleRGB=true
   )
  ]
  rendererClasses [&NiagaraSpriteRendererProperties]
  renderers [{rendererClass=&NiagaraSpriteRendererProperties}]
 }
 Sparks_Activate {
  simTarget=GPUComputeSim
  emitterUpdateScript [
   &EmitterState(Life Cycle Mode=::NewEnumerator1, Loop Behavior=::NewEnumerator1, Loop Delay=0.1, Loop Duration=0.2, UseLoopDelay=true)
   &SpawnRate(
    Spawn Probability=&MakeCustomFloatFromBool(Bool=$User.Enable Blade Arc, True Float=1, False Float=0)
    SpawnRate=200
    Use Spawn Probability=true
   )
  ]
  particleSpawnScript [
   &InitializeParticle(
    Color=$User.Color_Blade_A
    Lifetime Max=0.4
    Lifetime Min=0.1
    Lifetime Mode=::NewEnumerator1
    Position Offset=&Lerp_Vector(
     A=<0 0 12>
     B=<0 0 100>
     Alpha=&FloatFromCurve(
      FloatCurve=DataInterface() {propertyValues {Curve {keys [{interpMode=RCIM_Cubic, tangentMode=RCTM_User, time=0.15, value=0}, {interpMode=RCIM_Cubic, tangentMode=RCTM_User, time=0.3, value=1}]}}}
      CurveIndex=$Emitter.Age
     )
    )
    Position Offset Coordinate Space=::Local
    Sprite Size Mode=::NewEnumerator1
    Uniform Sprite Size Max=1
    Uniform Sprite Size Min=0.75
    UsePositionOffset=true
    Write Material Random=true
   )
   &AddVelocity(
    Cone Angle=&RandomRangeFloat(Minimum=0, Maximum=0.1)
    Cone Angle Mode:ENiagara_AngleInput=::NewEnumerator1
    Cone Axis=<0 0 1>
    Module.UseConeFalloff=false
    Module.UseVelDistribution=false
    Override Seed=false
    Velocity Mode=::NewEnumerator2
    Velocity Speed=&RandomRangeFloat(Minimum=100, Maximum=500)
   )
  ]
  particleUpdateScript [
   &ParticleState()
   &Drag(Drag=10, Use Linear Drag=true, Use Rotational Drag=false)
   &CurlNoiseForce(Noise Frequency=5000, Noise Strength=2000, Pan Noise=false)
   &SolveForcesAndVelocity(Clamp Velocity=false, Limit Acceleration=false)
   &ScaleSpriteSizeBySpeed(Max Scale Factor=<0.75 8>, Velocity Threshold=400)
   &ScaleColor(
    Scale Alpha=&FloatFromCurve(FloatCurve=DataInterface() {propertyValues {Curve {keys [{interpMode=RCIM_Cubic, tangentMode=RCTM_User, time=0.5, value=1}, {interpMode=RCIM_Cubic, tangentMode=RCTM_User, time=1, value=0}]}}})
    Scale RGB=&Lerp_Vector(
     A=<1 1 1>
     B=<4 4 4>
     Alpha=&FloatFromCurve(FloatCurve=DataInterface() {propertyValues {Curve {keys [{interpMode=RCIM_Cubic, tangentMode=RCTM_User, time=0, value=1}, {interpMode=RCIM_Cubic, tangentMode=RCTM_User, time=0.66, value=0}]}}})
    )
    ScaleA=true
    ScaleRGB=true
   )
  ]
  rendererClasses [&NiagaraSpriteRendererProperties]
  renderers [{rendererClass=&NiagaraSpriteRendererProperties}]
 }
}
summary {
 systemName=NS_Galileo_Lobster_Rezzout
 emitters [
  {emitterName=Blade_Glow, rendererClasses [&NiagaraSpriteRendererProperties, &NiagaraLightRendererProperties]}
  {emitterName=Blade_Glow_Activate, rendererClasses [&NiagaraSpriteRendererProperties]}
  {emitterName=Blade_Arcs, rendererClasses [&NiagaraRibbonRendererProperties, &NiagaraSpriteRendererProperties]}
  {emitterName=Sparks_Activate, simTarget=GPUComputeSim, rendererClasses [&NiagaraSpriteRendererProperties]}
  {emitterName=Sparks, simTarget=GPUComputeSim, rendererClasses [&NiagaraSpriteRendererProperties]}
  {
   emitterName=Blade_Flare_Activate
   rendererClasses [&NiagaraSpriteRendererProperties, &NiagaraLightRendererProperties]
  }
  {
   bEnabled=false
   emitterName=Blade_Glow001
   rendererClasses [&NiagaraSpriteRendererProperties, &NiagaraLightRendererProperties]
  }
  {
   emitterName=Blade_Flare_Activate001
   rendererClasses [&NiagaraSpriteRendererProperties, &NiagaraLightRendererProperties]
  }
  {emitterName=Saber, rendererClasses [&NiagaraSpriteRendererProperties, &NiagaraMeshRendererProperties]}
  {
   emitterName=Glow_Circle_Base
   rendererClasses [&NiagaraSpriteRendererProperties, &NiagaraLightRendererProperties]
  }
 ]
 userVariables [
  User.bisFrontEndPreview {type=NiagaraBool, defaultValue=true}
  User.Color_Blade_A {type=LinearColor, description="", defaultValue {struct=LinearColor, value=<rgba 2 0.5 0.25 1>}}
  User.Color_Blade_B {type=LinearColor, description="", defaultValue {struct=LinearColor, value=<rgba 1 0.0833 0 1>}}
  User.Color_Glow {type=LinearColor, description="", defaultValue {struct=LinearColor, value=<rgba 2 0 0 1>}}
  User.CurveSaberAll {
   type=NiagaraDataInterfaceCurve
   description=""
   defaultValue {
    struct=NiagaraExt_VariableValue_DataInterface
    value {
     dataInterface=&self:NiagaraDataInterfaceCurve_0
     dataInterfaceClass=&NiagaraDataInterfaceCurve
    }
   }
  }
  User.Enable Blade Arc {type=NiagaraBool}
  User.LoopDelayAll {type=NiagaraFloat, defaultValue=0, description=""}
 ]
}
