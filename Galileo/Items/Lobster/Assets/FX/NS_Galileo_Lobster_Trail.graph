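# Niagara ribbon trail for the Galileo Lobster item's blade, spawned from a user-supplied skeletal mesh while AnimTrailActivate is set.
# Custom AnimTrail modules spawn, initialize and update very short-lived particles, which gain upward velocity only in front-end preview.
# Over lifetime, color shifts from Blade_A to Blade_B, then alpha fades out.
# Two ribbon renderers draw the trail, with separate glow color and slightly wider glow width attributes.
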NiagaraSystem /Galileo/Items/Lobster/Assets/FX/NS_Galileo_Lobster_Trail
dependencies {usedDataInterfaces [&NiagaraDataInterfaceCurve], usedDynamicInputs [], usedModules [&SystemState]}
emitters {Wyeth_AnimTrails {
  emitterSpawnScript [/SetVariables(Emitter.Glow_Leading_Edge=10) {isSetParameters=true}]
  emitterUpdateScript [
   &EmitterState(Life Cycle Mode=::NewEnumerator1, Loop Duration=0.5, UseLoopDelay=false)
   &self:SpawnAnimTrail(AnimTrail is Active:NiagaraBool=$User.AnimTrailActivate, Mesh:NiagaraDataInterfaceSkeletalMesh=$User.SkeletalMesh, Socket B=1) {module=SpawnAnimTrail}
  ]
  particleSpawnScript [
   &self:InitializeAnimTrail(Color:LinearColor=<rgba 1 1 1 4>, Lifetime=0.0667, Ribbon Width Scale=1) {module=InitializeAnimTrail}
   &AddVelocity(Velocity=&MakeCustomVectorFromBool(Bool=$User.bisFrontEndPreview, True Vector=<0 0 300>, False Vector=<0 0 0>))
  ]
  particleUpdateScript [
   &ParticleState() {module=UpdateAge}
   &SolveForcesAndVelocity(Clamp Velocity=false, Limit Acceleration=false)
   &self:UpdateAnimTrail() {module=UpdateAnimTrail}
   &ScaleColor(
    Scale Alpha=&FloatFromCurve(FloatCurve=DataInterface() {propertyValues {Curve {keys [{arriveTangent=-3.6374, interpMode=RCIM_Cubic, time=0.5, value=1}, <at 0.6 0 RCIM_Cubic>]}}})
    Scale RGB=&Lerp_Vector(
     A=&MakeVectorFromLinearColorRGB(Color=$User.Color_Blade_B)
     B=&MakeVectorFromLinearColorRGB(Color=$User.Color_Blade_A)
     Alpha=&FloatFromCurve(FloatCurve=DataInterface() {propertyValues {Curve {keys [{arriveTangent=-5.3518, interpMode=RCIM_Cubic, time=0.2, value=1}, <at 0.5 0 RCIM_Cubic>]}}})
    )
    ScaleA=true
    ScaleRGB=true
   )
   /SetVariables(Particles.Color_Glow:LinearColor=$User.Color_Glow) {isSetParameters=true}
   /SetVariables(Particles.RibbonWidth_Glow:NiagaraFloat=&Multiply_Float(A=$Particles.RibbonWidth, B=1.15)) {isSetParameters=true}
  ]
  rendererClasses [&NiagaraRibbonRendererProperties]
  renderers [
   {rendererClass=&NiagaraRibbonRendererProperties}
   {rendererClass=&NiagaraRibbonRendererProperties, rendererIndex=1}
  ]
 }}
summary {
 systemName=NS_Galileo_Lobster_Trail
 emitters [{emitterName=Wyeth_AnimTrails, rendererClasses [&NiagaraRibbonRendererProperties]}]
 userVariables [
  User.AnimTrailActivate {type=NiagaraBool}
  User.bisFrontEndPreview {type=NiagaraBool}
  User.Color_Blade_A {type=LinearColor, description="", defaultValue {struct=LinearColor, value=<rgba 0.75 2.6253 3 1>}}
  User.Color_Blade_B {type=LinearColor, description="", defaultValue {struct=LinearColor, value=<rgba 0 0.8335 1 1>}}
  User.Color_Glow {type=LinearColor, description="", defaultValue {struct=LinearColor, value=<rgba 0 0.2 2 1>}}
  User.RibbonIDOffset {type=NiagaraInt32}
  User.SkeletalMesh {
   type=NiagaraDataInterfaceSkeletalMesh
   defaultValue {
    struct=NiagaraExt_VariableValue_DataInterface
    value {dataInterface=&self:NiagaraDataInterfaceSkeletalMesh_0}
   }
  }
 ]
}
