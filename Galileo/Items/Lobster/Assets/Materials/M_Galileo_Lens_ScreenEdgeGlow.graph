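# Unlit translucent particle material for the Galileo Lobster lens screen-edge glow, usable on sprites and Niagara meshes.
# An inverted radial SphereMask, shaped by Radius, Hardness and Exponent, multiplies an inverted square edge mask to form the glow.
# Color and particle color tint the mask, while a DebugTimeSine pulse and particle alpha modulate a separate alpha-style product.
# Leftover nodes include an "Old Particle Version" dual-sphere mask and lerps on a Mesh parameter.
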Material /Galileo/Items/Lobster/Assets/Materials/M_Galileo_Lens_ScreenEdgeGlow
graph {
 properties {
  bUsedWithNiagaraMeshParticles=true
  bUsedWithNiagaraSprites=true
  bUsedWithParticleSprites=true
  blendMode=BLEND_Translucent
  pixelDepthOffsetMode=PDOM_Legacy
  shadingModel=MSM_Unlit
  usedShadingModels=Unlit
 }
 nodes [
  Constant()
  Constant() {r=0.5}
  n1 = ScalarParameter() {parameterName=Mesh}
  LinearInterpolate(A=Constant() {r=3}, Alpha=@n1) {constAlpha=0, constB=1.5}
  LinearInterpolate(Alpha=@n1) {constA=0.5}
  n2 = Power(Base=SphereMask(
    A=TextureCoordinate()
    B=Constant2Vector() {g=0.5, r=0.5}
    Radius=ScalarParameter() {defaultValue=0.3, parameterName=Radius}
    Hardness=ScalarParameter() {defaultValue=0.85, parameterName=Hardness}
   ) {attenuationRadius=0.5, hardnessPercent=0})
  n3 = OneMinus(None=MaterialFunctionCall!(UVs=TextureCoordinate(), Mult=Constant() {r=1.5}, Exp=Constant() {r=1}) {materialFunction=/Game/Effects/Fort_Effects/MaterialFunctions/MF_SquareEdgeMask}.SquareMask)
  n4 = Multiply(A=Power(Base=OneMinus(None=@n2), Exp=ScalarParameter() {defaultValue=1.5, parameterName=Exponent}), B=@n3)
  n5 = MaterialFunctionCall(Speed=Constant() {r=2.2}) {materialFunction=&DebugTimeSine}
  n6 = ParticleColor()
  Multiply(A=Multiply(A=@n4, B=LinearInterpolate(Alpha=@n5.Result) {constA=0.1, constB=0.5}) {constB=0.3}, B=@n6.A)
  Multiply(A=VectorParameter() {parameterName=Color, defaultValue=<rgba 1 1 1 0>}.RGB, B=Multiply(A=@n6.RGB, B=@n4))
  n7 = TextureCoordinate() {uTiling=0.5}
  n8 = Constant2Vector() {g=0.5, r=0.25}
  n9 = TextureCoordinate() {vTiling=0.5}
  n10 = Constant2Vector() {g=0.25, r=0.5}
  n11 = Multiply(
   A=OneMinus(None=SphereMask(A=@n7, B=@n8) {attenuationRadius=0.74, hardnessPercent=0})
   B=OneMinus(None=SphereMask(A=@n9, B=@n10) {attenuationRadius=0.74, hardnessPercent=0})
  )
  Multiply(A=@n11, B=@n3) {desc=Old Particle Version}
  Power(Base=@n11) {constExponent=1.5}
 ]
}
